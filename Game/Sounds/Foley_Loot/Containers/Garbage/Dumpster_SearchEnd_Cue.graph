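# Sound cue for the end of searching a dumpster for loot, layering a no-more-loot UI sound with dumpster search foley.
# A mixer blends the UI wave with a random pick of three search waves, shaped by volume and pitch envelopes that fade it out, then slightly randomized by a modulator.
# Prop-search attenuation spatializes it under the loot-collect sound class, with a concurrency override limiting simultaneous plays.
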SoundCue /Game/Sounds/Foley_Loot/Containers/Garbage/Dumpster_SearchEnd_Cue
root=Mixer_12
nodes [
 {
  id=Enveloper_11
  children [Random_19]
  properties {
   pitchCurve {editorCurveData {keys [{time=-0.0037}, <at 0.0001 0.9999>, <at 0.5998 0.7662>]}}
   volumeCurve {editorCurveData {keys [<at -0.001 0.0028>,<at 0.0213 0.5133>,<at 0.3216 0.2579>,<at 0.747 0.0028>]}}
  }
 }
 {id=Mixer_12, children [WavePlayer_43, Modulator_15], properties {inputVolume [0.8,0.8]}}
 {id=Modulator_15, children [Enveloper_11], properties {pitchMax=1.07, pitchMin=0.92, volumeMax=1.07, volumeMin=0.92}}
 {id=Random_19, children [WavePlayer_62, WavePlayer_63, WavePlayer_64], properties {weights [1,1,1]}}
 {id=WavePlayer_43, properties {soundWaveAssetPtr=/Game/Sounds/Fort_UI_Sounds/Gameplay/Fort_UI_NoMoreLoot}}
 {id=WavePlayer_62, properties {soundWaveAssetPtr=&/Dumpster_Search01}}
 {id=WavePlayer_63, properties {soundWaveAssetPtr=&/Dumpster_Search02}}
 {id=WavePlayer_64, properties {soundWaveAssetPtr=&/Dumpster_Search03}}
]
properties {
 attenuationSettings=/Game/Sounds/Attenuation/Fort_Prop_Search_Att
 bOverrideConcurrency=true
 maxDistance=1300
 soundClassObject=&Fort_Loot_Collect
 volumeMultiplier=0.5
 concurrencyOverrides {maxCount=4, resolutionRule=StopFarthestThenPreventNew}
}
