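# Sound cue for searching a dumpster loot container, routed to the loot collect sound class with prop-search attenuation.
# A mixer layers two random picks: one of two waves with slight pitch and volume modulation at full level, and one of three unmodulated waves at half level.
# A concurrency override prevents new instances from starting once the limit is reached, rather than cutting off playing ones.
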SoundCue /Game/Sounds/Foley_Loot/Containers/Garbage/Dumpster_Search_Cue
root=Mixer_1
nodes [
 {id=Mixer_1, children [Modulator_4,Random_3], properties {inputVolume [1,0.5]}}
 {id=Modulator_4, children [Random_20], properties {pitchMax=1.07, pitchMin=0.92, volumeMax=1.07, volumeMin=0.92}}
 {id=Random_20, children [WavePlayer_66, WavePlayer_19], properties {weights [1,1]}}
 {id=Random_3, children [WavePlayer_6,WavePlayer_7,WavePlayer_8], properties {weights [1,1,1]}}
 {id=WavePlayer_19, properties {soundWaveAssetPtr=&/Dumpster_Search03}}
 {id=WavePlayer_6, properties {soundWaveAssetPtr=&/Search_Dumpster_01}}
 {id=WavePlayer_66, properties {soundWaveAssetPtr=&/Dumpster_Search02}}
 {id=WavePlayer_7, properties {soundWaveAssetPtr=&/Search_Dumpster_02}}
 {id=WavePlayer_8, properties {soundWaveAssetPtr=&/Search_Dumpster_03}}
]
properties {
 attenuationSettings=/Game/Sounds/Attenuation/Fort_Prop_Search_Att
 bOverrideConcurrency=true
 maxDistance=1300
 soundClassObject=&Fort_Loot_Collect
 volumeMultiplier=0.65
 concurrencyOverrides {resolutionRule=PreventNew}
}
