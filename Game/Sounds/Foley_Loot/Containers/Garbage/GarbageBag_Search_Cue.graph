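# Sound cue for searching a garbage bag loot container, assigned to the Fort_Loot_Collect sound class.
# A Mixer layers a random pick of three garbage bag search waves with a random pick of three generic gear search waves.
# Only the bag layer passes through a Modulator, which raises its pitch with slight variation and halves its volume.
# Attenuation and concurrency are overridden in the cue, with concurrency set to prevent new instances at the limit.
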SoundCue /Game/Sounds/Foley_Loot/Containers/Garbage/GarbageBag_Search_Cue
root=Mixer_27
nodes [
 {id=Mixer_27, children [Modulator_5,Random_26], properties {inputVolume [0.9,0.85]}}
 {id=Modulator_5, children [Random_14], properties {pitchMax=1.6, pitchMin=1.4, volumeMax=0.5, volumeMin=0.5}}
 {id=Random_14, children [WavePlayer_54, WavePlayer_55, WavePlayer_56], properties {weights [1,1,1]}}
 {id=Random_26, children [WavePlayer_134, WavePlayer_135, WavePlayer_137], properties {weights [1,1,1]}}
 {id=WavePlayer_134, properties {soundWaveAssetPtr=/Game/Sounds/Foley_Loot/Containers/Misc/Prop_Gear_Search_05}}
 {id=WavePlayer_135, properties {soundWaveAssetPtr=/Game/Sounds/Foley_Loot/Containers/Misc/Prop_Gear_Search_06}}
 {id=WavePlayer_137, properties {soundWaveAssetPtr=/Game/Sounds/Foley_Loot/Containers/Misc/Prop_Gear_Search_01}}
 {id=WavePlayer_54, properties {soundWaveAssetPtr=&/garbagebag_search_01}}
 {id=WavePlayer_55, properties {soundWaveAssetPtr=&/garbagebag_search_02}}
 {id=WavePlayer_56, properties {soundWaveAssetPtr=&/garbagebag_search_03}}
]
properties {
 attenuationSettings=/Game/Sounds/Attenuation/Fort_Prop_Search_Att
 bOverrideAttenuation=true
 bOverrideConcurrency=true
 maxDistance=1500
 soundClassObject=&Fort_Loot_Collect
 volumeMultiplier=1
 attenuationOverrides {falloffDistance=1000, attenuationShapeExtents {x=500}}
 concurrencyOverrides {resolutionRule=PreventNew}
}
