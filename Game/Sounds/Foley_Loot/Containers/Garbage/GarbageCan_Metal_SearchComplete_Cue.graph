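# Sound cue for completing a search of a metal garbage can, a loot container foley sound.
# It randomly picks one of three equally weighted metal search waves, fades it out with a short volume envelope, then applies slight random pitch and volume modulation.
# It uses prop-search attenuation and the loot-collect sound class, with overridden concurrency that caps simultaneous plays, stopping the farthest first.
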SoundCue /Game/Sounds/Foley_Loot/Containers/Garbage/GarbageCan_Metal_SearchComplete_Cue
root=Modulator_2
nodes [
 {
  id=Enveloper_3
  children [Random_1]
  properties {pitchCurve {editorCurveData {keys [{time=0}]}}, volumeCurve {editorCurveData {keys [<at 0 1>,<at 0.65 0>]}}}
 }
 {id=Modulator_2, children [Enveloper_3], properties {pitchMax=1.07, pitchMin=0.92, volumeMax=1.07, volumeMin=0.92}}
 {id=Random_1, children [WavePlayer_3,WavePlayer_4,WavePlayer_5], properties {weights [1,1,1]}}
 {id=WavePlayer_3, properties {soundWaveAssetPtr=&/GarbageCan_Metal_Search01}}
 {id=WavePlayer_4, properties {soundWaveAssetPtr=&/GarbageCan_Metal_Search02}}
 {id=WavePlayer_5, properties {soundWaveAssetPtr=&/GarbageCan_Metal_Search03}}
]
properties {
 attenuationSettings=/Game/Sounds/Attenuation/Fort_Prop_Search_Att
 bOverrideConcurrency=true
 maxDistance=1300
 soundClassObject=&Fort_Loot_Collect
 volumeMultiplier=0.4
 concurrencyOverrides {maxCount=4, resolutionRule=StopFarthestThenPreventNew}
}
