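# Sound cue for searching a metal garbage can, classed as loot-collect audio with prop-search attenuation.
# A mixer layers one of three random metal can search waves, with slight pitch and volume modulation, under the cue's main level.
# A second, quieter layer plays a random garbage bag rustle after a fixed delay, shaped by a volume envelope that fades it out.
# A concurrency override blocks new instances once the limit is reached.
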SoundCue /Game/Sounds/Foley_Loot/Containers/Garbage/GarbageCan_Metal_Search_Cue
root=Mixer_6
nodes [
 {id=Delay_20, children [Enveloper_2], properties {delayMax=1, delayMin=1}}
 {
  id=Enveloper_2
  children [Random_8]
  properties {pitchCurve {editorCurveData {keys [{time=0}]}}, volumeCurve {editorCurveData {keys [<at 0 1>,<at 2 1>,<at 3 0>]}}}
 }
 {id=Mixer_6, children [Modulator_18,Delay_20], properties {inputVolume [1,0.5]}}
 {id=Modulator_18, children [Random_0], properties {pitchMax=1.07, pitchMin=0.92, volumeMax=1.07, volumeMin=0.92}}
 {id=Random_0, children [WavePlayer_0,WavePlayer_1,WavePlayer_2], properties {weights [1,1,1]}}
 {id=Random_8, children [WavePlayer_39, WavePlayer_40, WavePlayer_41], properties {weights [1,1,1]}}
 {id=WavePlayer_0, properties {soundWaveAssetPtr=&/GarbageCan_Metal_Search01}}
 {id=WavePlayer_1, properties {soundWaveAssetPtr=&/GarbageCan_Metal_Search02}}
 {id=WavePlayer_2, properties {soundWaveAssetPtr=&/GarbageCan_Metal_Search03}}
 {id=WavePlayer_39, properties {soundWaveAssetPtr=&/garbagebag_search_01}}
 {id=WavePlayer_40, properties {soundWaveAssetPtr=&/garbagebag_search_02}}
 {id=WavePlayer_41, properties {soundWaveAssetPtr=&/garbagebag_search_03}}
]
properties {
 attenuationSettings=/Game/Sounds/Attenuation/Fort_Prop_Search_Att
 bOverrideConcurrency=true
 maxDistance=1300
 soundClassObject=&Fort_Loot_Collect
 volumeMultiplier=0.5
 concurrencyOverrides {resolutionRule=PreventNew}
}
