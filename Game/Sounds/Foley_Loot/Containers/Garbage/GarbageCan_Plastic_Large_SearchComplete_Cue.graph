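# Sound cue for completing a search of a large plastic garbage can container, played through the loot collect sound class with prop search attenuation.
# Randomly picks one of three equally weighted search waves and shapes it with a volume envelope that rises sharply, then fades out within a second.
# A modulator adds slight pitch and volume variation per play.
# Overridden concurrency caps simultaneous instances, stopping the farthest first.
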SoundCue /Game/Sounds/Foley_Loot/Containers/Garbage/GarbageCan_Plastic_Large_SearchComplete_Cue
root=Modulator_15
nodes [
 {
  id=Enveloper_8
  children [Random_5]
  properties {
   pitchCurve {editorCurveData {keys [{time=-0.0038}, <at 0.0001 0.9999>]}}
   volumeCurve {editorCurveData {keys [<at -0.001 0.0028>,<at 0.0377 0.6808>,<at 0.3216 0.2579>,<at 0.747 0.0028>]}}
  }
 }
 {id=Modulator_15, children [Enveloper_8], properties {pitchMax=1.07, pitchMin=0.92, volumeMax=1.07, volumeMin=0.92}}
 {id=Random_5, children [WavePlayer_34, WavePlayer_36, WavePlayer_37], properties {weights [1,1,1]}}
 {id=WavePlayer_34, properties {soundWaveAssetPtr=&/GarbageCan_Plastic_Large_Search01}}
 {id=WavePlayer_36, properties {soundWaveAssetPtr=&/GarbageCan_Plastic_Large_Search03}}
 {id=WavePlayer_37, properties {soundWaveAssetPtr=&/GarbageCan_Plastic_Large_Search04}}
]
properties {
 attenuationSettings=/Game/Sounds/Attenuation/Fort_Prop_Search_Att
 bOverrideConcurrency=true
 maxDistance=1300
 soundClassObject=&Fort_Loot_Collect
 volumeMultiplier=0.7
 concurrencyOverrides {maxCount=4, resolutionRule=StopFarthestThenPreventNew}
}
