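# Sound cue for searching a large plastic garbage can, routed to the loot collect sound class with prop search attenuation.
# A mixer layers one of three random can search waves, with slight pitch and volume modulation, over a half-volume garbage bag rustle.
# The bag layer picks from three random waves, starts after a one-second delay and fades out through a volume envelope.
# A concurrency override uses the prevent-new resolution rule.
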SoundCue /Game/Sounds/Foley_Loot/Containers/Garbage/GarbageCan_Plastic_Large_Search_Cue
root=Mixer_7
nodes [
 {id=Delay_21, children [Enveloper_3], properties {delayMax=1, delayMin=1}}
 {
  id=Enveloper_3
  children [Random_9]
  properties {pitchCurve {editorCurveData {keys [{time=0}]}}, volumeCurve {editorCurveData {keys [<at 0 1>,<at 2 1>,<at 3 0>]}}}
 }
 {id=Mixer_7, children [Modulator_18,Delay_21], properties {inputVolume [1,0.5]}}
 {id=Modulator_18, children [Random_4], properties {pitchMax=1.07, pitchMin=0.92, volumeMax=1.07, volumeMin=0.92}}
 {id=Random_4, children [WavePlayer_26, WavePlayer_27, WavePlayer_30], properties {weights [1,1,1]}}
 {id=Random_9, children [WavePlayer_42, WavePlayer_43, WavePlayer_44], properties {weights [1,1,1]}}
 {id=WavePlayer_26, properties {soundWaveAssetPtr=&/GarbageCan_Plastic_Large_Search07}}
 {id=WavePlayer_27, properties {soundWaveAssetPtr=&/GarbageCan_Plastic_Large_Search01}}
 {id=WavePlayer_30, properties {soundWaveAssetPtr=&/GarbageCan_Plastic_Large_Search04}}
 {id=WavePlayer_42, properties {soundWaveAssetPtr=&/garbagebag_search_01}}
 {id=WavePlayer_43, properties {soundWaveAssetPtr=&/garbagebag_search_02}}
 {id=WavePlayer_44, properties {soundWaveAssetPtr=&/garbagebag_search_03}}
]
properties {
 attenuationSettings=/Game/Sounds/Attenuation/Fort_Prop_Search_Att
 bOverrideConcurrency=true
 maxDistance=1300
 soundClassObject=&Fort_Loot_Collect
 volumeMultiplier=0.7
 concurrencyOverrides {resolutionRule=PreventNew}
}
