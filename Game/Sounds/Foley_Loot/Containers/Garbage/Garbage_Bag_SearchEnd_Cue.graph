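# Sound cue for the end of a garbage bag container search.
# A mixer layers the NoMoreLoot UI sound with a quieter, randomly picked garbage bag search rustle.
# The rustle fades out through a volume envelope, then a modulator raises its pitch by a fixed amount and slightly varies its volume.
# It uses the Loot_Collect sound class with overridden attenuation and a concurrency cap that stops the farthest instance first.
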SoundCue /Game/Sounds/Foley_Loot/Containers/Garbage/Garbage_Bag_SearchEnd_Cue
root=Mixer_2
nodes [
 {
  id=Enveloper_1
  children [Random_17]
  properties {pitchCurve {editorCurveData {keys [{time=0}]}}, volumeCurve {editorCurveData {keys [<at 0 1>,<at 1.5 0>]}}}
 }
 {id=Mixer_2, children [WavePlayer_3,Modulator_3], properties {inputVolume [0.8,0.5]}}
 {id=Modulator_3, children [Enveloper_1], properties {pitchMax=1.5, pitchMin=1.5, volumeMax=1.07, volumeMin=0.92}}
 {id=Random_17, children [WavePlayer_63, WavePlayer_64, WavePlayer_65], properties {weights [1,1,1]}}
 {id=WavePlayer_3, properties {soundWaveAssetPtr=/Game/Sounds/Fort_UI_Sounds/Gameplay/Fort_UI_NoMoreLoot}}
 {id=WavePlayer_63, properties {soundWaveAssetPtr=&/garbagebag_search_01}}
 {id=WavePlayer_64, properties {soundWaveAssetPtr=&/garbagebag_search_02}}
 {id=WavePlayer_65, properties {soundWaveAssetPtr=&/garbagebag_search_03}}
]
properties {
 attenuationSettings=/Game/Sounds/Attenuation/Fort_Prop_Search_Att
 bOverrideAttenuation=true
 bOverrideConcurrency=true
 maxDistance=1500
 soundClassObject=&Fort_Loot_Collect
 attenuationOverrides {falloffDistance=1000, attenuationShapeExtents {x=500}}
 concurrencyOverrides {maxCount=4, resolutionRule=StopFarthestThenPreventNew}
}
