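# Sound cue for the foley played when a bedroom trashcan search completes in loot containers.
# A random node picks evenly between two bathroom wastebasket search-end waves, then a modulator slightly varies pitch and volume.
# It plays quietly in the loot-collect sound class with prop-search attenuation and a concurrency override that stops the farthest instance first, then prevents new ones.
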SoundCue /Game/Sounds/Foley_Loot/Containers/Garbage/Trashcan_Bedroom_SearchComplete_Cue
root=Modulator_6
nodes [
 {id=Modulator_6, children [Random_6], properties {pitchMax=1.07, pitchMin=0.92, volumeMax=1.07, volumeMin=0.92}}
 {id=Random_6, children [WavePlayer_28, WavePlayer_29], properties {weights [1,1]}}
 {id=WavePlayer_28, properties {soundWaveAssetPtr=&/WasteBasket_Bathroom_SearchEnd3}}
 {id=WavePlayer_29, properties {soundWaveAssetPtr=&/WasteBasket_Bathroom_SearchEnd6}}
]
properties {
 attenuationSettings=/Game/Sounds/Attenuation/Fort_Prop_Search_Att
 bOverrideConcurrency=true
 maxDistance=1300
 soundClassObject=&Fort_Loot_Collect
 volumeMultiplier=0.3
 concurrencyOverrides {maxCount=4, resolutionRule=StopFarthestThenPreventNew}
}
