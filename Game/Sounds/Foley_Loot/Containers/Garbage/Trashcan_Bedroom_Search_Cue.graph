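# Sound cue for searching a bedroom trashcan, playing wastebasket foley on the loot-collect sound class with prop-search attenuation.
# A random node picks one of three search waves and feeds a modulator that slightly varies pitch and volume.
# The mixer plays that shared chain twice, immediately and again through a roughly one-second randomized delay at lower volume, layering a longer rummage.
# A concurrency override prevents new instances once the limit is reached.
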SoundCue /Game/Sounds/Foley_Loot/Containers/Garbage/Trashcan_Bedroom_Search_Cue
root=Mixer_32
nodes [
 {id=Delay_52, children [Modulator_10], properties {delayMax=1.2, delayMin=1}}
 {id=Mixer_32, children [Modulator_10,Delay_52], properties {inputVolume [1,0.7]}}
 {id=Modulator_10, children [Random_4], properties {pitchMax=1.07, pitchMin=0.92, volumeMax=1.07, volumeMin=0.92}}
 {id=Random_4, children [WavePlayer_18, WavePlayer_19, WavePlayer_20], properties {weights [1,1,1]}}
 {id=WavePlayer_18, properties {soundWaveAssetPtr=&/WasteBasket_Bedroom_Search5}}
 {id=WavePlayer_19, properties {soundWaveAssetPtr=&/WasteBasket_Bedroom_Search1}}
 {id=WavePlayer_20, properties {soundWaveAssetPtr=&/WasteBasket_Bedroom_Search2}}
]
properties {
 attenuationSettings=/Game/Sounds/Attenuation/Fort_Prop_Search_Att
 bOverrideConcurrency=true
 maxDistance=1300
 soundClassObject=&Fort_Loot_Collect
 volumeMultiplier=0.4
 concurrencyOverrides {resolutionRule=PreventNew}
}
